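# AudioButtonToggle is a UMG text toggle button for audio widgets, with selected and deselected color schemes.
# Clicking flips Selected through SetSelected or SetDeselected, which recolor background and label and broadcast OnSelected (optional via ReportEvent) or OnDeselected.
# PreConstruct applies TextValue, refreshes corners and calls SetDeselected.
# Four WidgetSwitchers choose square or curved corners from per-corner booleans, reusing one corner texture mirrored by negative render scale.
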WidgetBlueprint /AudioWidgets/AudioButtonToggle/AudioButtonToggle
graphs [
 EventGraph {nodes [
   "On Clicked (Button01)"() {is=title K2Node_ComponentBoundEvent}
   > n1 = Branch(Condition=Get Selected() -> (Selected:bool))
   > SetDeselected()
   SetSelected!(execute=@n1.else, self, ReportEvent:bool=true) -> (then)
   Event PreConstruct()
   > "SetText (Text)"(self=Get ValueText() -> (ValueText:Text Object Reference), InText=Get TextValue() -> (TextValue:Text))
   > RefreshCorners()
   > SetDeselected()
   Event Construct()
   Event Tick()
  ]}
 GetIsSelected {nodes [
   GetIsSelected() {is=title K2Node_FunctionEntry}
   > Return Node!(execute, IsSelected:bool=Get Selected())
  ]}
 OnDeselected {nodes [OnDeselected!() -> (then, Button:Audio Button Toggle Object Reference) {is=title K2Node_FunctionEntry}]}
 OnSelected {nodes [OnSelected!() -> (then, Button:Audio Button Toggle Object Reference) {is=title K2Node_FunctionEntry}]}
 RefreshCorners {nodes [
   RefreshCorners() {is=title K2Node_FunctionEntry}
   > SetActiveWidgetIndex(
    self=Get WidgetSwitchLeftTop() -> (WidgetSwitchLeftTop:Widget Switcher Object Reference)
    Index="To Integer (Boolean)"!(InBool:bool=Get CurveTopLeft() -> (CurveTopLeft:bool)) -> (ReturnValue:int=0)
   )
   > SetActiveWidgetIndex(
    self=Get WidgetSwitchRightCurveTop() -> (WidgetSwitchRightCurveTop:Widget Switcher Object Reference)
    Index="To Integer (Boolean)"(InBool=Get CurveTopRight() -> (CurveTopRight:bool)) -> (ReturnValue=0)
   )
   > SetActiveWidgetIndex(
    self=Get WidgetSwitchLeftBottom() -> (WidgetSwitchLeftBottom:Widget Switcher Object Reference)
    Index="To Integer (Boolean)"(InBool=Get CurveBottomLeft() -> (CurveBottomLeft:bool)) -> (ReturnValue=0)
   )
   > SetActiveWidgetIndex(
    self=Get WidgetSwitchRightCurveBottom() -> (WidgetSwitchRightCurveBottom:Widget Switcher Object Reference)
    Index="To Integer (Boolean)"(InBool=Get CurveBottomRight() -> (CurveBottomRight:bool)) -> (ReturnValue=0)
   )
  ]}
 SetBackgroundColor {nodes [
   n1 = SetBackgroundColor!() -> (then, NewColor:Linear Color Structure) {is=title K2Node_FunctionEntry}
   > SetColorAndOpacity(self:Image Object Reference=Get LeftBoxBottom() -> (LeftBoxBottom:Image Object Reference), InColorAndOpacity:Linear Color Structure=@n1)
   > SetColorAndOpacity(self=Get LeftBoxTop() -> (LeftBoxTop:Image Object Reference), InColorAndOpacity=@n1)
   > SetColorAndOpacity(self=Get LeftCurveBottom() -> (LeftCurveBottom:Image Object Reference), InColorAndOpacity=@n1)
   > SetColorAndOpacity(self=Get LeftCurveTop() -> (LeftCurveTop:Image Object Reference), InColorAndOpacity=@n1)
   > SetColorAndOpacity(self=Get RightBoxBottom() -> (RightBoxBottom:Image Object Reference), InColorAndOpacity=@n1)
   > SetColorAndOpacity(self=Get RightBoxTop() -> (RightBoxTop:Image Object Reference), InColorAndOpacity=@n1)
   > SetColorAndOpacity(self=Get RightCurveBottom() -> (RightCurveBottom:Image Object Reference), InColorAndOpacity=@n1)
   > SetColorAndOpacity(self=Get RightCurveTop() -> (RightCurveTop:Image Object Reference), InColorAndOpacity=@n1)
   > SetBrushColor(self=Get TextBorder() -> (TextBorder:Border Object Reference), InBrushColor=@n1)
  ]}
 SetDeselected {nodes [
   SetDeselected() {is=title K2Node_FunctionEntry}
   > Branch(Condition=Get Selected())
   > Set Selected(Selected:bool=false) -> (Output_Get:bool)
   > SetBackgroundColor!(execute, self, NewColor:Linear Color Structure=Get BackgroundColor() -> (BackgroundColor:Linear Color Structure)) -> (then) {is=title K2Node_CallFunction}
   > SetColorAndOpacity(self:Text Object Reference=Get ValueText(), InColorAndOpacity:Slate Color Structure=Get TextColor() -> (TextColor:Slate Color Structure))
   > Call On Deselected!(execute:Exec, self:Audio Button Toggle Object Reference, Button:Audio Button Toggle Object Reference=Self-Reference()) -> (then:Exec) {is=title K2Node_CallDelegate}
  ]}
 SetSelected {nodes [
   SetSelected!() -> (then, ReportEvent:bool=true) {is=title K2Node_FunctionEntry}
   > n1 = Branch(Condition=Get Selected())
   Set Selected(execute=@n1.else, Selected=true)
   > SetBackgroundColor(NewColor=Get BackgroundColorSelected() -> (BackgroundColorSelected:Linear Color Structure))
   > SetColorAndOpacity(self=Get ValueText(), InColorAndOpacity=Get TextColorSelected() -> (TextColorSelected:Slate Color Structure))
   > Branch(Condition=Get ReportEvent() -> (ReportEvent:bool))
   > Call On Selected!(execute:Exec, self:Audio Button Toggle Object Reference, Button:Audio Button Toggle Object Reference=Self-Reference()) -> (then:Exec) {is=title K2Node_CallDelegate}
  ]}
]
variables [
 BackgroundColor {type=LinearColor, default=<rgba 0.0395 0.0395 0.0395 1>}
 BackgroundColorSelected {type=LinearColor, default=<rgba 0.0103 0.0103 0.0103 1>}
 CurveBottomLeft {type=boolean, default=true}
 CurveBottomRight {type=boolean, default=true}
 CurveTopLeft {type=boolean, default=true}
 CurveTopRight {type=boolean, default=true}
 OnDeselected {}
 OnSelected {}
 Selected {type=boolean}
 TextColor {type=SlateColor, default {colorUseRule=UseColor_Specified, specifiedColor=<rgba 1 1 1 0.75>}}
 TextColorSelected {type=SlateColor, default {colorUseRule=UseColor_Specified, specifiedColor=<rgba 0.0194 0.4969 1 1>}}
 TextValue {type=string, default=Text01}
]
widgets {
 info {inheritedWidgetCount=0, parentClass=&UserWidget, rootWidgetClass=&Button, widgetCount=18}
 widgets [
  Button01 {
   widgetClassPath=&Button
   widget=Button01
   properties {
    backgroundColor=<rgba 1 1 1 0>
    widgetStyle {
     clickedSlateSound {resourceObject=null}
     disabled {
      bIsDynamicallyLoaded=false
      drawAs=NoDrawType
      imageType=NoImage
      mirroring=NoMirror
      resourceName=None
      resourceObject=null
      tiling=NoTile
      imageSize=<0 0>
      margin=<ltrb 0 0 0 0>
      outlineSettings {
       bUseBrushTransparency=false
       roundingType=HalfHeightRadius
       width=0
       color {colorUseRule=UseColor_Specified, specifiedColor=<rgba 0 0 0 0>}
       cornerRadii=<0 0 0 0>
      }
      tintColor {colorUseRule=UseColor_Specified, specifiedColor=<rgba 1 1 1 1>}
      uVRegion {bIsValid=false, max=<0 0>, min=<0 0>}
     }
     disabledForeground {colorUseRule=UseColor_Specified, specifiedColor=<rgba 1 1 1 1>}
     hovered {
      bIsDynamicallyLoaded=false
      drawAs=NoDrawType
      imageType=NoImage
      mirroring=NoMirror
      resourceName=None
      resourceObject=null
      tiling=NoTile
      imageSize=<32 32>
      margin=<ltrb 0 0 0 0>
      outlineSettings {
       bUseBrushTransparency=true
       roundingType=FixedRadius
       width=1
       color {colorUseRule=UseColor_Specified, specifiedColor=<rgba 0.7243 0.7243 0.7243 1>}
       cornerRadii=<4 4 4 4>
      }
      tintColor {colorUseRule=UseColor_Specified, specifiedColor=<rgba 0.7243 0.7243 0.7243 1>}
      uVRegion {bIsValid=false, max=<0 0>, min=<0 0>}
     }
     hoveredForeground {colorUseRule=UseColor_Specified, specifiedColor=<rgba 0 1 1 1>}
     hoveredSlateSound {resourceObject=null}
     normal {
      bIsDynamicallyLoaded=false
      drawAs=Box
      imageType=NoImage
      mirroring=NoMirror
      resourceName=None
      resourceObject=null
      tiling=NoTile
      imageSize=<32 32>
      margin=<ltrb 0 0 0 0>
      outlineSettings {
       bUseBrushTransparency=true
       roundingType=FixedRadius
       width=1
       color {colorUseRule=UseColor_Specified, specifiedColor=<rgba 0.6951 0.6951 0.6951 1>}
       cornerRadii=<4 4 4 4>
      }
      tintColor {colorUseRule=UseColor_Specified, specifiedColor=<rgba 1 1 1 1>}
      uVRegion {bIsValid=false, max=<0 0>, min=<0 0>}
     }
     normalForeground {colorUseRule=UseColor_Specified, specifiedColor=<rgba 0 1 1 1>}
     normalPadding=<ltrb 0 0 0 0>
     pressed {
      bIsDynamicallyLoaded=false
      drawAs=Box
      imageType=NoImage
      mirroring=NoMirror
      resourceName=None
      resourceObject=null
      tiling=NoTile
      imageSize=<32 32>
      margin=<ltrb 0 0 0 0>
      outlineSettings {
       bUseBrushTransparency=true
       roundingType=FixedRadius
       width=1
       color {colorUseRule=UseColor_Specified, specifiedColor=<rgba 0.7243 0.7243 0.7243 1>}
       cornerRadii=<4 4 4 4>
      }
      tintColor {colorUseRule=UseColor_Specified, specifiedColor=<rgba 0.0103 0.0103 0.0103 1>}
      uVRegion {bIsValid=false, max=<0 0>, min=<0 0>}
     }
     pressedForeground {colorUseRule=UseColor_Specified, specifiedColor=<rgba 0.0103 0.0103 0.0103 1>}
     pressedPadding=<ltrb 0 0 0 0>
     pressedSlateSound {resourceObject=null}
    }
   }
  }
  HorizontalContentBox {
   widgetClassPath=&HorizontalBox
   parent=Button01
   widget=HorizontalContentBox
   properties {visibility=HitTestInvisible}
   slot {class=ButtonSlot, properties {horizontalAlignment=HAlign_Fill, verticalAlignment=VAlign_Fill, padding=<ltrb 0 0 0 0>}}
  }
  LeftCurveVerticalBox {widgetClassPath=&VerticalBox, parent=HorizontalContentBox, widget=LeftCurveVerticalBox, slot {class=HorizontalBoxSlot}}
  WidgetSwitchLeftTop {
   widgetClassPath=&WidgetSwitcher
   bIsVariable=true
   parent=LeftCurveVerticalBox
   widget=WidgetSwitchLeftTop
   slot {class=VerticalBoxSlot, properties {size {sizeRule=Fill, value=0.5}}}
  }
  LeftBoxTop {
   widgetClassPath=&Image
   bIsVariable=true
   parent=WidgetSwitchLeftTop
   widget=LeftBoxTop
   properties {
    brush {
     bIsDynamicallyLoaded=false
     drawAs=Box
     imageType=NoImage
     mirroring=NoMirror
     resourceName=None
     resourceObject=null
     tiling=NoTile
     imageSize=<8 12>
     margin=<ltrb 0 0 0 0>
     outlineSettings {
      bUseBrushTransparency=false
      roundingType=HalfHeightRadius
      width=0
      color {colorUseRule=UseColor_Specified, specifiedColor=<rgba 0 0 0 0>}
      cornerRadii=<0 0 0 0>
     }
     tintColor {colorUseRule=UseColor_Specified, specifiedColor=<rgba 1 1 1 1>}
     uVRegion {bIsValid=false, max=<0 0>, min=<0 0>}
    }
    colorAndOpacity=<rgba 0.0395 0.0395 0.0395 1>
    renderTransform {angle=0, scale=<-1 1>, shear=<0 0>, translation=<0 0>}
   }
   slot {class=WidgetSwitcherSlot}
  }
  LeftCurveTop {
   widgetClassPath=&Image
   bIsVariable=true
   parent=WidgetSwitchLeftTop
   widget=LeftCurveTop
   properties {
    brush {
     bIsDynamicallyLoaded=false
     drawAs=Image
     imageType=NoImage
     mirroring=NoMirror
     resourceName=None
     resourceObject=&/AudioButton_Corner4x
     tiling=NoTile
     imageSize=<8 12>
     margin=<ltrb 0 0 0 0>
     outlineSettings {
      bUseBrushTransparency=false
      roundingType=HalfHeightRadius
      width=0
      color {colorUseRule=UseColor_Specified, specifiedColor=<rgba 0 0 0 0>}
      cornerRadii=<0 0 0 0>
     }
     tintColor {colorUseRule=UseColor_Specified, specifiedColor=<rgba 1 1 1 1>}
     uVRegion {bIsValid=false, max=<0 0>, min=<0 0>}
    }
    colorAndOpacity=<rgba 0.0395 0.0395 0.0395 1>
   }
   slot {class=WidgetSwitcherSlot}
  }
  WidgetSwitchLeftBottom {
   widgetClassPath=&WidgetSwitcher
   bIsVariable=true
   parent=LeftCurveVerticalBox
   widget=WidgetSwitchLeftBottom
   slot {class=VerticalBoxSlot, properties {size {sizeRule=Fill, value=0.5}}}
  }
  LeftBoxBottom {
   widgetClassPath=&Image
   bIsVariable=true
   parent=WidgetSwitchLeftBottom
   widget=LeftBoxBottom
   properties {
    brush {
     bIsDynamicallyLoaded=false
     drawAs=Box
     imageType=NoImage
     mirroring=NoMirror
     resourceName=None
     resourceObject=null
     tiling=NoTile
     imageSize=<8 12>
     margin=<ltrb 0 0 0 0>
     outlineSettings {
      bUseBrushTransparency=false
      roundingType=HalfHeightRadius
      width=0
      color {colorUseRule=UseColor_Specified, specifiedColor=<rgba 0 0 0 0>}
      cornerRadii=<0 0 0 0>
     }
     tintColor {colorUseRule=UseColor_Specified, specifiedColor=<rgba 1 1 1 1>}
     uVRegion {bIsValid=false, max=<0 0>, min=<0 0>}
    }
    colorAndOpacity=<rgba 0.0395 0.0395 0.0395 1>
    renderTransform {angle=0, scale=<-1 1>, shear=<0 0>, translation=<0 0>}
   }
   slot {class=WidgetSwitcherSlot}
  }
  LeftCurveBottom {
   widgetClassPath=&Image
   bIsVariable=true
   parent=WidgetSwitchLeftBottom
   widget=LeftCurveBottom
   properties {
    brush {
     bIsDynamicallyLoaded=false
     drawAs=Image
     imageType=NoImage
     mirroring=NoMirror
     resourceName=None
     resourceObject=&AudioButton_Corner4x
     tiling=NoTile
     imageSize=<8 12>
     margin=<ltrb 0 0 0 0>
     outlineSettings {
      bUseBrushTransparency=false
      roundingType=HalfHeightRadius
      width=0
      color {colorUseRule=UseColor_Specified, specifiedColor=<rgba 0 0 0 0>}
      cornerRadii=<0 0 0 0>
     }
     tintColor {colorUseRule=UseColor_Specified, specifiedColor=<rgba 1 1 1 1>}
     uVRegion {bIsValid=false, max=<0 0>, min=<0 0>}
    }
    colorAndOpacity=<rgba 0.0395 0.0395 0.0395 1>
    renderTransform {angle=0, scale=<1 -1>, shear=<0 0>, translation=<0 0>}
   }
   slot {class=WidgetSwitcherSlot}
  }
  TextBorder {
   widgetClassPath=&Border
   bIsVariable=true
   parent=HorizontalContentBox
   widget=TextBorder
   properties {verticalAlignment=VAlign_Center, brushColor=<rgba 0.0395 0.0395 0.0395 1>, padding=<ltrb 0 0 0 0>}
   slot {class=HorizontalBoxSlot, properties {size {sizeRule=Fill, value=1}}}
  }
  ValueText {
   widgetClassPath=&TextBlock
   bIsVariable=true
   parent=TextBorder
   widget=ValueText
   properties {
    justification=Center
    text=Unset
    colorAndOpacity {colorUseRule=UseColor_Specified, specifiedColor=<rgba 1 1 1 0.75>}
    font {
     bForceMonospaced=false
     bMaterialIsStencil=false
     fontMaterial=null
     fontObject=&Roboto
     letterSpacing=0
     monospacedWidth=1
     size=10
     skewAmount=0
     typefaceFontName=Regular
     outlineSettings {
      bApplyOutlineToDropShadows=false
      bMiteredCorners=false
      bSeparateFillAlpha=false
      outlineMaterial=null
      outlineSize=0
      outlineColor=<rgba 0 0 0 1>
     }
    }
   }
   slot {class=BorderSlot, properties {verticalAlignment=VAlign_Center, padding=<ltrb 0 0 0 0>}}
  }
  RightCurveVerticalBox {
   widgetClassPath=&VerticalBox
   parent=HorizontalContentBox
   widget=RightCurveVerticalBox
   properties {visibility=HitTestInvisible}
   slot {class=HorizontalBoxSlot}
  }
  WidgetSwitchRightCurveTop {
   widgetClassPath=&WidgetSwitcher
   bIsVariable=true
   parent=RightCurveVerticalBox
   widget=WidgetSwitchRightCurveTop
   slot {class=VerticalBoxSlot, properties {size {sizeRule=Fill, value=0.5}}}
  }
  RightBoxTop {
   widgetClassPath=&Image
   bIsVariable=true
   parent=WidgetSwitchRightCurveTop
   widget=RightBoxTop
   properties {
    brush {
     bIsDynamicallyLoaded=false
     drawAs=Box
     imageType=NoImage
     mirroring=NoMirror
     resourceName=None
     resourceObject=null
     tiling=NoTile
     imageSize=<8 12>
     margin=<ltrb 0 0 0 0>
     outlineSettings {
      bUseBrushTransparency=false
      roundingType=HalfHeightRadius
      width=0
      color {colorUseRule=UseColor_Specified, specifiedColor=<rgba 0 0 0 0>}
      cornerRadii=<0 0 0 0>
     }
     tintColor {colorUseRule=UseColor_Specified, specifiedColor=<rgba 1 1 1 1>}
     uVRegion {bIsValid=false, max=<0 0>, min=<0 0>}
    }
    colorAndOpacity=<rgba 0.0395 0.0395 0.0395 1>
    renderTransform {angle=0, scale=<-1 1>, shear=<0 0>, translation=<0 0>}
   }
   slot {class=WidgetSwitcherSlot}
  }
  RightCurveTop {
   widgetClassPath=&Image
   bIsVariable=true
   parent=WidgetSwitchRightCurveTop
   widget=RightCurveTop
   properties {
    brush {
     bIsDynamicallyLoaded=false
     drawAs=Image
     imageType=NoImage
     mirroring=NoMirror
     resourceName=None
     resourceObject=&AudioButton_Corner4x
     tiling=NoTile
     imageSize=<8 12>
     margin=<ltrb 0 0 0 0>
     outlineSettings {
      bUseBrushTransparency=false
      roundingType=HalfHeightRadius
      width=0
      color {colorUseRule=UseColor_Specified, specifiedColor=<rgba 0 0 0 0>}
      cornerRadii=<0 0 0 0>
     }
     tintColor {colorUseRule=UseColor_Specified, specifiedColor=<rgba 1 1 1 1>}
     uVRegion {bIsValid=false, max=<0 0>, min=<0 0>}
    }
    colorAndOpacity=<rgba 0.0395 0.0395 0.0395 1>
    renderTransform {angle=0, scale=<-1 1>, shear=<0 0>, translation=<0 0>}
   }
   slot {class=WidgetSwitcherSlot}
  }
  WidgetSwitchRightCurveBottom {
   widgetClassPath=&WidgetSwitcher
   bIsVariable=true
   parent=RightCurveVerticalBox
   widget=WidgetSwitchRightCurveBottom
   slot {class=VerticalBoxSlot, properties {size {sizeRule=Fill, value=0.5}}}
  }
  RightBoxBottom {
   widgetClassPath=&Image
   bIsVariable=true
   parent=WidgetSwitchRightCurveBottom
   widget=RightBoxBottom
   properties {
    brush {
     bIsDynamicallyLoaded=false
     drawAs=Box
     imageType=NoImage
     mirroring=NoMirror
     resourceName=None
     resourceObject=null
     tiling=NoTile
     imageSize=<8 12>
     margin=<ltrb 0 0 0 0>
     outlineSettings {
      bUseBrushTransparency=false
      roundingType=HalfHeightRadius
      width=0
      color {colorUseRule=UseColor_Specified, specifiedColor=<rgba 0 0 0 0>}
      cornerRadii=<0 0 0 0>
     }
     tintColor {colorUseRule=UseColor_Specified, specifiedColor=<rgba 1 1 1 1>}
     uVRegion {bIsValid=false, max=<0 0>, min=<0 0>}
    }
    colorAndOpacity=<rgba 0.0395 0.0395 0.0395 1>
    renderTransform {angle=0, scale=<-1 1>, shear=<0 0>, translation=<0 0>}
   }
   slot {class=WidgetSwitcherSlot}
  }
  RightCurveBottom {
   widgetClassPath=&Image
   bIsVariable=true
   parent=WidgetSwitchRightCurveBottom
   widget=RightCurveBottom
   properties {
    brush {
     bIsDynamicallyLoaded=false
     drawAs=Image
     imageType=NoImage
     mirroring=NoMirror
     resourceName=None
     resourceObject=&AudioButton_Corner4x
     tiling=NoTile
     imageSize=<8 12>
     margin=<ltrb 0 0 0 0>
     outlineSettings {
      bUseBrushTransparency=false
      roundingType=HalfHeightRadius
      width=0
      color {colorUseRule=UseColor_Specified, specifiedColor=<rgba 0 0 0 0>}
      cornerRadii=<0 0 0 0>
     }
     tintColor {colorUseRule=UseColor_Specified, specifiedColor=<rgba 1 1 1 1>}
     uVRegion {bIsValid=false, max=<0 0>, min=<0 0>}
    }
    colorAndOpacity=<rgba 0.0395 0.0395 0.0395 1>
    renderTransform {angle=0, scale=<-1 -1>, shear=<0 0>, translation=<0 0>}
   }
   slot {class=WidgetSwitcherSlot}
  }
 ]
}
